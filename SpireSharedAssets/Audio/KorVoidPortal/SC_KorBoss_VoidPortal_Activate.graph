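# Sound cue for the Kor boss void portal activation, a layered one-shot in the boss fight SFX class.
# A mixer combines three layers: a start sound that plays immediately, a briefly delayed VividRazor Baron void activation, and an explosion that lands two seconds in.
# Only the Baron layer gets random pitch modulation.
# The cue uses its own attenuation asset and overrides concurrency to cap simultaneous plays.
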SoundCue /SpireSharedAssets/Audio/KorVoidPortal/SC_KorBoss_VoidPortal_Activate
root=Mixer_0
nodes [
 {id=Delay_0, children [Modulator_0], properties {delayMax=0.25, delayMin=0.25}}
 {id=Delay_1, children [WavePlayer_1], properties {delayMax=2, delayMin=2}}
 {id=Mixer_0, children [Delay_0,WavePlayer_2,Delay_1], properties {inputVolume [0.5,1,1.1]}}
 {id=Modulator_0, children [WavePlayer_5], properties {pitchMax=1.6, pitchMin=0.85}}
 {id=WavePlayer_1, properties {soundWaveAssetPtr=&/KorBoss_Activate_Void_Explo}}
 {id=WavePlayer_2, properties {soundWaveAssetPtr=&/KorBoss_Activate_Void_Start}}
 {id=WavePlayer_5, properties {soundWaveAssetPtr=&/VividRazor_Baron_Activate_Void}}
]
properties {
 attenuationSettings=&/Att_KorBoss_VoidPortal_Activate
 bOverrideConcurrency=true
 maxDistance=13000
 soundClassObject=&Fort_BossFight_SFX
 volumeMultiplier=1.3
 concurrencyOverrides {maxCount=3}
}
